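# Animated screen material for the large display on the Creative Class Designer device, routed through MF_CreativeMoveTool.
# An icon strip slides horizontally in eased quarter-UV steps on looping time, masked to a center band.
# It combines with a base level and a code texture that scrolls in stepped jumps, feeding tinted base color and power-curved emissive.
# A sphere mask around the player position blends moving scanlines into the emissive at close range.
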Material /Game/Creative/Devices/Class_Designer/Materials/M_Class_Designer_ScreenLarge
graph {
 properties {bUseMaterialAttributes=true, pixelDepthOffsetMode=PDOM_Legacy, usedShadingModels=Default Lit}
 nodes [
  n1 = SphereMask(
   A=ComponentMask(None=TextureCoordinate()) {r=true}
   B=ScalarParameter() {defaultValue=0.5, parameterName=IconMaskCenter}
   Radius=ScalarParameter() {defaultValue=0.27, parameterName=IconMaskRadius}
   Hardness=ScalarParameter() {defaultValue=0.9, parameterName=IconMaskHardness}
  )
  n2 = AppendVector(A=ScalarParameter() {defaultValue=0.125, parameterName=UBias}, B=ScalarParameter() {defaultValue=0.4, parameterName=VBias})
  n3 = Multiply(
   A=TextureCoordinate()
   B=AppendVector(A=ScalarParameter() {defaultValue=0.5, parameterName=UTile}, B=ScalarParameter() {defaultValue=2.2, parameterName=VTile})
  )
  n4 = Multiply(A=Time() {bOverride_Period=true, period=2}) {constB=0.5}
  n5 = Multiply(A=Power(Base=MaterialFunctionCall!(In=@n4) {materialFunction=/Game/Effects/Fort_Effects/Materials/_Functions/MF_Smoothstep_01}.Result) {constExponent=16}) {constB=0.25}
  n6 = Time() {bOverride_Period=true, period=8}
  n7 = Ceil(None=Fmod(A=Multiply(A=@n6) {constB=0.5}, B=Constant() {r=4}))
  n8 = Multiply(A=@n7) {constB=0.25}
  n9 = Add(A=Subtract(A=Add(A=@n5, B=@n8)) {constB=0.25}, B=@n7)
  n10 = Constant()
  n11 = TextureSample(UVs=Add(A=Add(A=@n2, B=@n3), B=AppendVector(A=@n9, B=@n10))) {texture=/Game/Creative/Devices/Class_Designer/Textures/T_Class_Designer_ScreenLarge}
  n12 = Time() {bOverride_Period=true, period=100}
  n13 = Multiply(A=@n12, B=ScalarParameter() {defaultValue=0.62, parameterName=TimeScaleA})
  n14 = Multiply(A=@n12, B=ScalarParameter() {defaultValue=2.233, parameterName=TimeScaleB})
  n15 = Panner(Coordinate=TextureCoordinate() {vTiling=0.25}, Time=Add(A=Ceil(None=@n13), B=Ceil(None=@n14))) {bFractionalPart=true, speedY=0.02}
  n16 = Add(
   A=Add(A=Multiply(A=@n1, B=@n11.R), B=ScalarParameter() {defaultValue=0.3, parameterName=BGBase})
   B=Multiply(
    A=TextureSample(UVs=@n15) {texture=/Game/Creative/Devices/RoundSettings/Textures/T_Round_Settings_ScreenCode}.R
    B=ScalarParameter() {defaultValue=0.5, parameterName=TextScale}
   )
  )
  n17 = Power(
   Base=Multiply(A=@n16, B=VectorParameter() {parameterName=EmissiveTint, defaultValue=<rgba 0.1512 1.2 1.1243 0>}.RGB)
   Exp=ScalarParameter() {defaultValue=4, parameterName=EmissiveExponent}
  )
  n18 = ScalarParameter() {defaultValue=6, parameterName=EmissiveScale}
  n19 = Constant() {r=1}
  n20 = ComponentMask(None=Panner(Coordinate=TextureCoordinate()) {speedY=0.02}) {g=true}
  n21 = ScalarParameter() {defaultValue=45, parameterName=ScanlineRepeat}
  n22 = SphereMask(
   A=Frac(None=Multiply(A=@n20, B=@n21))
   B=Constant() {r=0.5}
   Radius=ScalarParameter() {defaultValue=0.4, parameterName=ScanlineRadius}
   Hardness=ScalarParameter() {defaultValue=0.2, parameterName=ScanlineHarndess}
  )
  n23 = CollectionParameter() {collection=&FortniteMaterialParameters, parameterName=PlayerPosition}
  n24 = SphereMask(
   A=ComponentMask(None=@n23) {b=true, g=true, r=true}
   B=WorldPosition().XYZ
   Radius=ScalarParameter() {defaultValue=350, parameterName=Radius}
   Hardness=ScalarParameter() {defaultValue=80, parameterName=Hardness}
  ) {attenuationRadius=350, hardnessPercent=80}
  n25 = SetMaterialAttributes!(
   MaterialAttributes
   Base Color=Multiply(A=@n16, B=VectorParameter() {parameterName=BaseColorTint, defaultValue=<rgba 0.0318 0.0516 0.5399 0>}.RGB)
   Metallic=ScalarParameter() {parameterName=Metallic}
   Specular=ScalarParameter() {defaultValue=0.8, parameterName=Specular}
   Roughness=ScalarParameter() {defaultValue=0.1, parameterName=Roughness}
   Emissive Color=Multiply(A=Multiply(A=@n17, B=@n18), B=LinearInterpolate(A=@n19, B=@n22, Alpha=@n24))
  ) {attributeSetTypes [
    #1
    #2
    #3
    #4
    #5
   ]}
  MaterialFunctionCall(MA In=@n25) {materialFunction=&MF_CreativeMoveTool}
 ]
}
