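# Lit material for the Class Designer creative device's small screen, showing a tinted screen texture with glitch distortion and scrolling scanlines.
# Two summed sine waves, sharpened in a VertexInterpolator, pulse a panning-noise horizontal offset on the center-scaled UVs.
# The texture's red channel drives tinted base color and exponent-boosted emissive, which a repeating Frac and SphereMask band pattern modulates.
# The resulting material attributes pass through MF_CreativeMoveTool.
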Material /Game/Creative/Devices/Class_Designer/Materials/M_Class_Designer_ScreenSmall
graph {
 properties {
  bUseMaterialAttributes=true
  bUsedWithMeshDeformer=true
  bUsedWithSkeletalMesh=true
  pixelDepthOffsetMode=PDOM_Legacy
  usedShadingModels=Default Lit
 }
 nodes [
  n1 = TextureCoordinate()
  n2 = Panner(Coordinate=TextureCoordinate() {uTiling=0}) {bFractionalPart=true, speedX=0.025, speedY=0.3844}
  n3 = Time()
  n4 = Add(A=Sine(None=@n3) {period=0.355}, B=Sine(None=@n3) {period=0.8755})
  n5 = VertexInterpolator(VS=Power(Base=Saturate(None=@n4)) {constExponent=5})
  n6 = Multiply(
   A=AppendVector(A=TextureSample(UVs=@n2) {samplerType=SAMPLERTYPE_Normal, texture=/Game/Global/Textures/Default/Noise/64x64Noise}.R, B=Constant())
   B=Multiply(A=ScalarParameter() {defaultValue=0.03, parameterName=DistortScale}, B=ComponentMask(None=@n5.PS) {r=true})
  )
  n7 = ScalarParameter() {defaultValue=0.5, parameterName=RamirezScaleU}
  n8 = ScalarParameter() {defaultValue=1.02, parameterName=RamirezScaleV}
  n9 = TextureSample(UVs=MaterialFunctionCall(UVs=Add(A=@n1, B=@n6), Texture Scale=AppendVector(A=@n7, B=@n8)) {materialFunction=&ScaleUVsByCenter}.UVs) {texture=/Game/Creative/Devices/Class_Designer/Textures/T_Class_Designer_Screen}
  n10 = Power(
   Base=Multiply(A=@n9.R, B=VectorParameter() {parameterName=EmissiveTint, defaultValue=<rgba 0.1357 1.2 0.6057 0>}.RGB)
   Exp=ScalarParameter() {defaultValue=6, parameterName=EmissiveExponent}
  )
  n11 = ScalarParameter() {defaultValue=8, parameterName=EmissiveScale}
  n12 = Panner(Coordinate=TextureCoordinate()) {speedY=0.02}
  n13 = Frac(None=Multiply(A=ComponentMask(None=@n12) {g=true}, B=ScalarParameter() {defaultValue=25, parameterName=ScanlineRepeat}))
  n14 = Constant() {r=0.5}
  n15 = ScalarParameter() {defaultValue=0.4, parameterName=ScanlineRadius}
  n16 = ScalarParameter() {defaultValue=0.2, parameterName=ScanlineHarndess}
  n17 = SetMaterialAttributes!(
   MaterialAttributes
   Base Color=Multiply(A=@n9.R, B=VectorParameter() {parameterName=BaseColorTint, defaultValue=<rgba 0.068 0.4558 0.5538 0>}.RGB)
   Metallic=ScalarParameter() {parameterName=Metallic}
   Specular=ScalarParameter() {defaultValue=0.8, parameterName=Specular}
   Roughness=ScalarParameter() {defaultValue=0.1, parameterName=Roughness}
   Emissive Color=Multiply(A=Multiply(A=@n10, B=@n11), B=SphereMask(A=@n13, B=@n14, Radius=@n15, Hardness=@n16))
  ) {attributeSetTypes [
    #1
    #2
    #3
    #4
    #5
   ]}
  MaterialFunctionCall(MA In=@n17) {materialFunction=&MF_CreativeMoveTool}
 ]
}
